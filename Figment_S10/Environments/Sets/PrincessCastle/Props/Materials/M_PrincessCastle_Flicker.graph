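# Additive unlit material for a flickering, tinted glow on Princess Castle props, faking local light on surrounding surfaces.
# Two summed sines drive the flicker, with speed varied by actor position.
# A sphere mask around the pivot, tested against the scene behind, shapes the glow; SM5/SM6 add scene-normal shading and depth fade.
# The mesh faces and shifts toward the camera; forward and mobile paths output zero and collapse it to its origin.
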Material /Figment_S10/Environments/Sets/PrincessCastle/Props/Materials/M_PrincessCastle_Flicker
graph {
 properties {blendMode=BLEND_Additive, pixelDepthOffsetMode=PDOM_Legacy, shadingModel=MSM_Unlit, usedShadingModels=Unlit}
 nodes [
  n1 = Time()
  n2 = MaterialFunctionCall(Float3=ActorPositionWS()) {materialFunction=&BreakOutFloat3Components}
  n3 = Add(A=Add(A=@n2.R, B=@n2.R), B=@n2.B)
  n4 = LinearInterpolate(A=Multiply(A=@n1) {constB=0.7}, B=Multiply(A=@n1) {constB=1.5}, Alpha=Frac(None=Divide(A=@n3) {constB=50}))
  n5 = Add(A=Sine(None=@n4) {period=0.5651}, B=Sine(None=@n4) {period=0.442})
  n6 = ScalarParameter() {defaultValue=0.1, parameterName=FlickerScale}
  n7 = Multiply(A=Add(A=Multiply(A=@n5, B=@n6)), B=VectorParameter() {parameterName=Glow Color, defaultValue=<rgba 1 0.2676 0.0122 0>}.RGB)
  n8 = MaterialFunctionCall() {materialFunction=&ObjectPivotPoint}
  n9 = MaterialFunctionCall() {materialFunction=&WorldPositionBehindTranslucency}
  n10 = DotProduct(
   A=Normalize(VectorInput=Subtract(A=@n8.Object Pivot Location, B=@n9.Result))
   B=ComponentMask(None=SceneTexture() {sceneTextureId=PPI_WorldNormal}.Color) {b=true, g=true, r=true}
  )
  n11 = DepthFade(Opacity=Multiply(A=@n7, B=Saturate(None=@n10)), FadeDistance=ScalarParameter() {defaultValue=25, parameterName=FadeDistance})
  n12 = ObjectRadius()
  n13 = Multiply(
   A=FeatureLevelSwitch(Default=@n7, ES2=@n7, ES3_1=@n7, SM4_REMOVED=@n7, SM5=@n11, SM6=@n11)
   B=SphereMask(
    A=@n8.Object Pivot Location
    B=@n9.Result
    Radius=Multiply(B=@n12) {constA=0.75}
    Hardness=ScalarParameter() {defaultValue=0.1, parameterName=FlareHardness}
   )
  )
  n14 = Constant()
  ShadingPathSwitch(Default=@n13, Deferred=@n13, Forward=@n14, Mobile=@n14)
  n15 = Subtract(A=CameraPositionWS(), B=ObjectPositionWS())
  n16 = Add(
   A=MaterialFunctionCall!(
    WorldPosition=WorldPosition().XYZ
    PivotLocation
    WS Normals
    Custom Object Basis 1=Constant3Vector() {constant=<rgba 0 1 0 1>}
    Custom Object Basis 2=Constant3Vector() {constant=<rgba 0 0 -1 0>}
    Custom Object Basis 3
   ) {materialFunction=/Engine/Functions/Engine_MaterialFunctions02/WorldPositionOffset/AlignMeshToTheCamera}.World Position Offset
   B=Multiply(A=Normalize(VectorInput=@n15), B=ScalarParameter() {defaultValue=20, parameterName=CameraOffsetBias})
  )
  n17 = Subtract(A=ObjectPositionWS(), B=WorldPosition().XYZ)
  ShadingPathSwitch(Default=@n16, Deferred=@n16, Forward=@n17, Mobile=@n17)
 ]
}
